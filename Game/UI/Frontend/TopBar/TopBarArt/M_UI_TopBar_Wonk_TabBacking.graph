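# Translucent UI material that draws the skewed "wonk" backing shape for frontend top bar tabs, with animated selection.
# Four edge masks come procedurally from offset, rotated and clamped UV gradients, then combine via scalar dot products and scale by Alpha.
# AnimateSelection blends each edge's offset, rotation and sharpness, with a per-side RandomValue parameter varying the tilt.
# AnimateSelectionColor separately blends between Color_1 and Color_2.
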Material /Game/UI/Frontend/TopBar/TopBarArt/M_UI_TopBar_Wonk_TabBacking
graph {
 properties {
  bUsedWithStaticMesh=false
  blendMode=BLEND_Translucent
  materialDomain=MD_UI
  pixelDepthOffsetMode=PDOM_Legacy
  usedShadingModels=Default Lit
 }
 nodes [
  LinearInterpolate(
   A=VectorParameter() {parameterName=Color_1, defaultValue=<rgba 0 0.4167 1 1>}.RGB
   B=VectorParameter() {parameterName=Color_2, defaultValue=<rgba 0.975 0.975 0.975 1>}.RGB
   Alpha=ScalarParameter() {parameterName=AnimateSelectionColor}
  )
  n1 = ScalarParameter() {parameterName=Angle_Offset_Left_Start}
  n2 = ScalarParameter() {defaultValue=-0.0075, parameterName=Angle_Offset_Left_Stop}
  n3 = ScalarParameter() {parameterName=AnimateSelection}
  n4 = ScalarParameter() {defaultValue=0.0018, parameterName=Angle_Rotation_Left}
  n5 = ScalarParameter() {parameterName=RandomValue_Left}
  n6 = Rotator(
   Coordinate=Add(A=TextureCoordinate(), B=LinearInterpolate(A=@n1, B=@n2, Alpha=@n3))
   Time=LinearInterpolate(A=Constant(), B=Add(A=@n4, B=@n5), Alpha=@n3)
  ) {speed=6.2832}
  n7 = ScalarParameter() {defaultValue=512, parameterName=U_Softness_Start}
  n8 = ScalarParameter() {defaultValue=176, parameterName=U_Softness_Stop}
  n9 = LinearInterpolate(A=@n7, B=@n8, Alpha=@n3)
  n10 = Clamp(None=Multiply(A=ComponentMask(None=@n6) {r=true}, B=Clamp(None=@n9, Min=@n8, Max=@n7)))
  n11 = TextureCoordinate()
  n12 = ScalarParameter() {parameterName=Angle_Offset_Right_Start}
  n13 = ScalarParameter() {defaultValue=-0.01, parameterName=Angle_Offset_Right_Stop}
  n14 = ScalarParameter() {defaultValue=-0.0025, parameterName=Angle_Rotation_Right}
  n15 = ScalarParameter() {parameterName=RandomValue_Right}
  n16 = Rotator(
   Coordinate=Add(A=OneMinus(None=@n11), B=LinearInterpolate(A=@n12, B=@n13, Alpha=@n3))
   Time=LinearInterpolate(A=Constant(), B=Add(A=@n14, B=@n15), Alpha=@n3)
  ) {speed=6.2832}
  n17 = ScalarParameter() {defaultValue=512, parameterName=U_Softness_Start}
  n18 = ScalarParameter() {defaultValue=176, parameterName=U_Softness_Stop}
  n19 = LinearInterpolate(A=@n17, B=@n18, Alpha=@n3)
  n20 = Clamp(None=Multiply(A=ComponentMask(None=@n16) {r=true}, B=Clamp(None=@n19, Min=@n18, Max=@n17)))
  n21 = ScalarParameter() {defaultValue=-0.16, parameterName=Angle_Offset_Top_Start}
  n22 = ScalarParameter() {defaultValue=-0.05, parameterName=Angle_Offset_Top_Stop}
  n23 = ScalarParameter() {defaultValue=0.005, parameterName=Angle_Rotation_Top}
  n24 = ScalarParameter() {parameterName=RandomValue_Top}
  n25 = Rotator(
   Coordinate=Add(A=TextureCoordinate(), B=LinearInterpolate(A=@n21, B=@n22, Alpha=@n3))
   Time=LinearInterpolate(A=Constant(), B=Add(A=@n23, B=@n24), Alpha=@n3)
  ) {speed=6.2832}
  n26 = ScalarParameter() {defaultValue=256, parameterName=V_Softness_Start}
  n27 = ScalarParameter() {defaultValue=56, parameterName=V_Softness_Stop}
  n28 = LinearInterpolate(A=@n26, B=@n27, Alpha=@n3)
  n29 = Clamp(None=Multiply(A=ComponentMask(None=@n25) {g=true}, B=Clamp(None=@n28, Min=@n27, Max=@n26)))
  n30 = TextureCoordinate()
  n31 = ScalarParameter() {defaultValue=-0.16, parameterName=Angle_Offset_Bottom_Start}
  n32 = ScalarParameter() {defaultValue=-0.05, parameterName=Angle_Offset_Bottom_Stop}
  n33 = ScalarParameter() {defaultValue=-0.0075, parameterName=Angle_Rotation_Bottom}
  n34 = ScalarParameter() {parameterName=RandomValue_Bottom}
  n35 = Rotator(
   Coordinate=Add(A=OneMinus(None=@n30), B=LinearInterpolate(A=@n31, B=@n32, Alpha=@n3))
   Time=LinearInterpolate(A=Constant(), B=Add(A=@n33, B=@n34), Alpha=@n3)
  ) {speed=6.2832}
  n36 = ScalarParameter() {defaultValue=256, parameterName=V_Softness_Start}
  n37 = ScalarParameter() {defaultValue=56, parameterName=V_Softness_Stop}
  n38 = LinearInterpolate(A=@n36, B=@n37, Alpha=@n3)
  n39 = Clamp(None=Multiply(A=ComponentMask(None=@n35) {g=true}, B=Clamp(None=@n38, Min=@n37, Max=@n36)))
  n40 = DotProduct(A=Saturate(None=DotProduct(A=@n10, B=@n20)), B=Saturate(None=DotProduct(A=@n29, B=@n39)))
  Saturate(None=Multiply(A=Saturate(None=@n40), B=ScalarParameter() {defaultValue=1, parameterName=Alpha}) {constB=0})
 ]
}
